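# Niagara vehicle boost effect for the Dagger cosmetic, named as a backup, emitting from skeletal mesh sockets, one per boost port.
# Three emitters are live: velocity-driven ribbon trails, cone mesh/sprites, and a mesh/sprite element with sine-driven wobble.
# Painted-variant modules supply per-color tints, while painted and frontend-preview flags switch colors, fades and camera offset.
# Earlier cone and sprite emitter variants remain but are disabled.
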NiagaraSystem /VehicleCosmetics/Boosts/Boost_Dagger/FX/NS_Boost_Dagger_Backup
dependencies {
 usedDataInterfaces [
  &NiagaraDataInterfaceCurve
  &NiagaraDataInterfaceColorCurve
  &NiagaraDataInterfaceArrayFloat
  &NiagaraDataInterfaceSkeletalMesh
 ]
 usedDynamicInputs []
 usedModules [
  &self:SystemSpawnScript.SystemScriptSource.SystemScriptGraph.NiagaraNodeAssignment_0.SetVariables_#1
  &SystemState
 ]
}
emitters {
 NE_VC_Boost_BIGCONE {
  bEnabled=false
  emitterUpdateScript [
   &EmitterState(Inactive Response=::NewEnumerator1, Life Cycle Mode=::NewEnumerator1, Loop Duration=1, UseLoopDelay=false)
   &SpawnBurst_Instantaneous(Spawn Count=$System.Initial.PortCount, Use Spawn Probability=false, UseLoopCountLimit=false)
   &VC_FilteredSocketCount()
   &SpawnRate(SpawnRate=8, Use Spawn Probability=false)
  ]
  particleSpawnScript [
   &VC_PaintedVariantsOverride(
    Black=<rgba 0.012 0.012 0.012 7>
    BurntSienna=<rgba 0.07 0.0212 0.0035 4>
    Cobalt=<rgba 0.012 0.012 0.6 2>
    Crimson=<rgba 0.25 0.0375 0.0871 4>
    DefaultColor=<rgba 1 0.2625 0.25 1>
    ForestGreen=<rgba 0 0.28 0.0141 5>
    Grey=<rgba 0.1 0.1 0.1 2.5>
    Lime=<rgba 0.1833 0.5 0.025 2>
    Orange=<rgba 0.5 0.12 0.025 2>
    Pink=<rgba 0.8 0.12 0.5733 2>
    Purple=<rgba 0.1133 0 0.6795 8>
    R1 BurntSienna:LinearColor=<rgba 0.3 0.091 0.015 1>
    R1 Cobalt:LinearColor=<rgba 0 0 1.5 1.5>
    R1 Crimson:LinearColor=<rgba 2 0.04 0.1053 1>
    R1 DefaultColor=<rgba 1 0.5 0.5 1>
    R1 Grey:LinearColor=<rgba 0.2 0.2 0.2 1.5>
    R1 Lime:LinearColor=<rgba 0.3333 1 0 1>
    R1 Orange:LinearColor=<rgba 1.5 0.3125 0.075 1>
    R1 Pink:LinearColor=<rgba 2 0.2396 1.3598 1>
    R1 Purple:LinearColor=<rgba 0.4167 0 1 1>
    R1 Saffron:LinearColor=<rgba 1 0.9167 0 1>
    R1 SkyBlue:LinearColor=<rgba 0 0.375 1.5 1>
    Saffron=<rgba 0.275 0.3 0 3>
    SkyBlue=<rgba 0 0.2 0.8 4>
    TitaniumWhite=<rgba 0.5 0.5 0.5 2>
    Use R1 Painted Overrides=true
   )
   /SetVariables(Particles.RENDERTAG:NiagaraInt32=&SpawnGroup()) {isSetParameters=true, module=SetVariables_#2}
   &InitializeParticle(
    Color=$Particles.Color
    Lifetime=0.5
    Mesh Scale=<0.62 0.138 0.138>
    Mesh Scale Mode=::NewEnumerator4
    Sprite Rotation Mode=::NewEnumerator2
    Sprite Size Mode=::NewEnumerator3
    UsePositionOffset=false
    Write Material Random=true
   )
   &PartitionParticles(Number Of Partitions=$System.PortCount)
   &SkeletalMeshLocation(
    Bone / Socket Index=$Particles.PartitionParticles.Partition
    Bone / Socket Sampling Mode=::NewEnumerator5
    MaskSpawnGroup=false
    Mesh Sampling Type=::NewEnumerator6
    Sampled Position Offset=&AddVectorToPosition(Position=<-40 0 0>, Vector=$User.DTCarOffset)
    Skeletal Mesh=$System.SkeletalMesh
   )
   &AddVelocity(Velocity=<-1500 0 0>) {enabled=false}
   &InitialMeshOrientation(Mesh Orientation Mode=::NewEnumerator2, Orientation Vector=<-1 0 0>, Use Rotation Vector=false)
   &MeshRotationForce(Rotation=<5 0 0>) {enabled=false}
   &ApplyInitialForces()
   /SetVariables(Particles.SIZE2:Vector3f=<0.11 0.052 0.052>) {isSetParameters=true, module=SetVariables_#3}
  ]
  particleUpdateScript [
   &ParticleState()
   &SolveForcesAndVelocity(Clamp Velocity=false, Limit Acceleration=false)
  ]
  rendererClasses [&NiagaraMeshRendererProperties, &NiagaraSpriteRendererProperties]
  renderers [
   {rendererClass=&NiagaraMeshRendererProperties, rendererIndex=0}
   {rendererClass=&NiagaraMeshRendererProperties, rendererIndex=1}
   {rendererClass=&NiagaraSpriteRendererProperties, rendererIndex=2}
  ]
 }
 NE_VC_Boost_BIGCONE002 {
  emitterUpdateScript [
   &EmitterState(Inactive Response=::NewEnumerator1, Life Cycle Mode=::NewEnumerator1, Loop Duration=1, UseLoopDelay=false)
   &SpawnBurst_Instantaneous(Spawn Count=$System.Initial.PortCount, Use Spawn Probability=false, UseLoopCountLimit=false)
   &VC_FilteredSocketCount()
   &SpawnRate(SpawnRate=8, Use Spawn Probability=false)
  ]
  particleSpawnScript [
   &VC_PaintedVariantsOverride(
    Black=<rgba 0.075 0.075 0.075 1>
    BurntSienna=<rgba 0.18 0.097 0.0585 1>
    Cobalt=<rgba 0.075 0.0998 0.5 0.35>
    Crimson=<rgba 0.4 0.14 0.1833 0.5>
    DefaultAlpha=0.1
    DefaultColor=<rgba 0.65 0.3412 0.325 0.15>
    ForestGreen=<rgba 0 0.45 0.0525 0.5>
    Grey=<rgba 0.21 0.21 0.21 0.2>
    Lime=<rgba 0.2817 0.65 0.0975 0.175>
    Orange=<rgba 0.5 0.2133 0.1 0.75>
    Pink=<rgba 0.6 0.18 0.46 0.5>
    Purple=<rgba 0.2521 0 0.55 0.55>
    R1 Black:LinearColor=<rgba 0.09 0.09 0.09 0.25>
    R1 BurntSienna:LinearColor=<rgba 0.45 0.2501 0.1575 0.35>
    R1 Cobalt:LinearColor=<rgba 0 0.0958 1.15 0.35>
    R1 Crimson:LinearColor=<rgba 0.85 0.1275 0.2479 0.45>
    R1 DefaultAlpha=0.1
    R1 DefaultColor=<rgba 1 0.6617 0.65 0.1>
    R1 ForestGreen:LinearColor=<rgba 0.185 0.5 0.185 0.45>
    R1 Grey:LinearColor=<rgba 0.45 0.45 0.45 0.15>
    R1 Lime:LinearColor=<rgba 0.575 1 0.15 0.075>
    R1 Orange:LinearColor=<rgba 1.25 0.4688 0.3125 0.15>
    R1 Pink:LinearColor=<rgba 2 0.4 1.4182 0.1>
    R1 Purple:LinearColor=<rgba 0.5833 0 1 0.25>
    R1 Saffron:LinearColor=<rgba 1 0.9167 0 0.15>
    R1 SkyBlue:LinearColor=<rgba 0 0.375 1.5 0.5>
    R1 TitaniumWhite:LinearColor=<rgba 0.7 0.7 0.7 0.45>
    Saffron=<rgba 0.4162 0.45 0.045 0.65>
    SkyBlue=<rgba 0.0475 0.3182 0.95 0.6>
    TitaniumWhite=<rgba 0.5 0.5 0.5 0.5>
    Use R1 Painted Overrides=true
   )
   /SetVariables(Particles.RENDERTAG:NiagaraInt32=&SpawnGroup()) {isSetParameters=true, module=SetVariables_#2}
   &InitializeParticle(
    Color=$Particles.Color
    Lifetime=0.5
    Mesh Scale=<0.62 0.138 0.138>
    Mesh Scale Mode=::NewEnumerator4
    Sprite Rotation Mode=::NewEnumerator2
    Sprite Size Mode=::NewEnumerator3
    UsePositionOffset=false
    Write Material Random=true
   )
   &PartitionParticles(Number Of Partitions=$System.PortCount)
   &SkeletalMeshLocation(
    Bone / Socket Index=$Particles.PartitionParticles.Partition
    Bone / Socket Sampling Mode=::NewEnumerator5
    MaskSpawnGroup=false
    Mesh Sampling Type=::NewEnumerator6
    Sampled Position Offset=&AddVectorToPosition(Position=<-40 0 0>, Vector=$User.DTCarOffset)
    Skeletal Mesh=$System.SkeletalMesh
   )
   &AddVelocity(Velocity=<-1500 0 0>) {enabled=false}
   &InitialMeshOrientation(Mesh Orientation Mode=::NewEnumerator2, Orientation Vector=<-1 0 0>, Use Rotation Vector=false)
   &MeshRotationForce(Rotation=<5 0 0>) {enabled=false}
   &ApplyInitialForces()
   /SetVariables(Particles.SIZE2:Vector3f=<0.11 0.052 0.052>) {isSetParameters=true, module=SetVariables_#3}
  ]
  particleUpdateScript [
   &ParticleState()
   &SolveForcesAndVelocity(Clamp Velocity=false, Limit Acceleration=false)
  ]
  rendererClasses [&NiagaraMeshRendererProperties, &NiagaraSpriteRendererProperties]
  renderers [
   {rendererClass=&NiagaraMeshRendererProperties, rendererIndex=0}
   {rendererClass=&NiagaraMeshRendererProperties, rendererIndex=1}
   {rendererClass=&NiagaraSpriteRendererProperties, rendererIndex=2}
  ]
 }
 NE_VC_Boost_RIBBONS {
  emitterUpdateScript [
   &EmitterState(
    Inactive Response=::NewEnumerator1
    Life Cycle Mode=::NewEnumerator1
    Loop Duration=1
    Recalculate Duration Each Loop=true
    UseLoopDelay=false
   )
   &VC_FilteredSocketCount()
   &VC_RateQualityHandler()
   &VC_SpawnRibbon_v1(Number of Ribbons=$Emitter.PortCount, Spawn Time=-1, SpawnRate=50)
  ]
  particleSpawnScript [
   &VC_PaintedVariantsOverride(
    Black=<rgba 0.012 0.012 0.012 1>
    BurntSienna=<rgba 0.03 0.0116 0.006 0.95>
    Cobalt=<rgba 0.003 0.003 0.06 1>
    Crimson=<rgba 0.1 0.01 0.0175 1>
    ForestGreen=<rgba 0 0.07 0.0035 1>
    Grey=<rgba 0.02 0.02 0.02 0.8>
    Lime=<rgba 0.0167 0.05 0 1>
    Orange=<rgba 0.08 0.028 0 1>
    Pink=<rgba 0.06 0.021 0.06 1>
    Purple=<rgba 0.0375 0 0.15 1>
    R1 Black:LinearColor=<rgba 0.011 0.011 0.011 1>
    R1 BurntSienna:LinearColor=<rgba 0.08 0.0271 0.0078 1>
    R1 Cobalt:LinearColor=<rgba 0.02 0.0327 0.4 1>
    R1 Crimson:LinearColor=<rgba 0.2 0.02 0.032 1>
    R1 ForestGreen:LinearColor=<rgba 0 0.2 0.0101 1>
    R1 Grey:LinearColor=<rgba 0.1 0.1 0.1 0.35>
    R1 Lime:LinearColor=<rgba 0.125 0.3 0 1>
    R1 Orange:LinearColor=<rgba 0.5 0.1833 0.025 0.5>
    R1 Pink:LinearColor=<rgba 0.3 0.09 0.3 0.5>
    R1 Purple:LinearColor=<rgba 0.1 0 0.4 1>
    R1 Saffron:LinearColor=<rgba 0.25 0.25 0 1>
    R1 SkyBlue:LinearColor=<rgba 0 0.0867 0.4 1>
    R1 TitaniumWhite:LinearColor=<rgba 1 1 1 0.15>
    R2 Black:LinearColor=<rgba 0.03 0.03 0.03 0.5>
    R2 BurntSienna:LinearColor=<rgba 0.1 0.037 0.01 0.8>
    R2 Cobalt:LinearColor=<rgba 0.008 0.0113 0.4 1>
    R2 Crimson:LinearColor=<rgba 0.2 0.02 0.029 0.4>
    R2 ForestGreen:LinearColor=<rgba 0 0.15 0.0076 1>
    R2 Grey:LinearColor=<rgba 0.07 0.07 0.07 0.5>
    R2 Lime:LinearColor=<rgba 0.1183 0.2 0.06 0.5>
    R2 Orange:LinearColor=<rgba 0.35 0.0992 0 0.8>
    R2 Pink:LinearColor=<rgba 0.3 0.12 0.3 0.3>
    R2 Purple:LinearColor=<rgba 0.1 0 0.4 0.8>
    R2 Saffron:LinearColor=<rgba 0.3667 0.4 0 0.3>
    R2 SkyBlue:LinearColor=<rgba 0 0.1341 0.5365 0.5>
    R2 TitaniumWhite:LinearColor=<rgba 0.3 0.3 0.3 0.5>
    Saffron=<rgba 0.0583 0.07 0 1>
    SkyBlue=<rgba 0 0.034 0.12 1>
    TitaniumWhite=<rgba 0.096 0.1044 0.12 1>
    Use R1 Painted Overrides=true
    Use R2 Painted Overrides=true
   )
   &VC_InitializeRibbon_v1(
    Color=$Particles.Color
    Custom Offset=true
    Enable Skeletal Position=true
    Local Vector Offset=&AddVectorToPosition(Position=<-15 0 0>, Vector=$User.DTCarOffset)
    Offset Amount=1
    Ribbon Width=100
   )
   &AddVelocity(Velocity=<4000 0 50>, Velocity Speed Scale=-1) {module=AddVelocity001}
   /VehicleCosmetics/FX/Niagara/Modules/MergeEmitters/VC_ExtraBindingColors(
    ExtraColors=2
    R1 Color:LinearColor=&MakeCustomLinearColorFromBool(Bool=$User.bIsPainted, True Color=$Particles.R1 Color, False Color=<rgba 1.25 0.3125 1.0156 1>)
    R2 Color:LinearColor=&MakeCustomLinearColorFromBool(Bool=$User.bIsPainted, True Color=$Particles.R2 Color, False Color=<rgba 3 0.975 0.3 1>)
   )
  ]
  particleUpdateScript [
   &ParticleState()
   &InheritSourceMovement(
    Applied Position Scale=&Multiply_VectorByFloat(
     Vector=<1 1 1>
     Float=&FloatFromCurve(FloatCurve=DataInterface() {propertyValues {Curve {keys [{interpMode=RCIM_Cubic, tangentMode=RCTM_User, time=0.2, value=1}, {interpMode=RCIM_Cubic, tangentMode=RCTM_User, time=1, value=0}]}}})
    )
    ApplyForce=false
    ApplyPosition=true
    ApplyPositionFalloff=false
   )
   &SolveForcesAndVelocity(Clamp Velocity=false, Limit Acceleration=false)
   &VC_UpdateRibbon_v1(Enable Skeletal Position=true)
   &DynamicMaterialParameters(
    Index 0 Param 1=&FloatFromCurve(FloatCurve=DataInterface() {propertyValues {Curve {keys [<at 0 1>,<at 1 0.5>]}}})
    Index 1 Param 1=&MakeCustomFloatFromBool(
     Bool=$User.bIsPainted
     True Float=&FloatFromCurve(
      FloatCurve=DataInterface() {propertyValues {Curve {keys [{arriveTangent=-6.5862, interpMode=RCIM_Cubic, leaveTangent=-6.5862, tangentMode=RCTM_User, time=0.1, value=1}, <at 0.5 0.1 RCIM_Cubic>]}}}
      Scale Curve=&MakeCustomFloatFromBool(Bool=$User.bIsFrontendPreview, True Float=10, False Float=5)
     )
     False Float=&MakeCustomFloatFromBool(Bool=$User.bIsFrontendPreview, True Float=1, False Float=2)
    )
    Index 1 Param 2=1
    Index 1 Param 3=0.5
    Index 1 Param 4=-1
    Index 2 Param 1=1
    Module.Index0Param0Write=true
    Module.Index0Param1Write=true
    Module.Index0Param2Write=true
    Module.Index0Param3Write=true
    Module.Index1Param0Write=true
    Module.Index1Param1Write=true
    Module.Index1Param2Write=true
    Module.Index1Param3Write=true
    Module.Index2Param0Write=true
    Module.Index2Param1Write=true
    Module.Index2Param2Write=true
    Module.Index2Param3Write=true
    Write Parameter Index 1=true
    Write Parameter Index 2=true
   )
   &ScaleColor(
    Scale Alpha=&FloatFromCurve(
     FloatCurve=DataInterface() {propertyValues {Curve {keys [<at 0.5 1 RCIM_Cubic>, {arriveTangentWeight=0.7803, interpMode=RCIM_Cubic, tangentMode=RCTM_User, time=0.9, value=0}]}}}
     Scale Curve=&MakeCustomFloatFromBool(Bool=$User.bIsFrontendPreview, True Float=2, False Float=6)
    )
    Scale RGB=&Multiply_VectorByFloat(
     Vector=<1 1 1>
     Float=&FloatFromCurve(
      FloatCurve=DataInterface() {propertyValues {Curve {keys [{arriveTangent=-0.1571, interpMode=RCIM_Cubic, leaveTangent=-0.1571, tangentMode=RCTM_User, time=0, value=2}, <at 0.7 0 RCIM_Cubic>]}}}
      Scale Curve=&MakeCustomFloatFromBool(
       Bool=$User.bIsFrontendPreview
       True Float=6
       False Float=&MakeCustomFloatFromBool(Bool=$User.bIsPainted, True Float=15, False Float=5)
      )
     )
    )
    ScaleA=true
    ScaleRGB=true
   )
   &Color(Color=&MakeCustomLinearColorFromBool(
     Bool=$User.bIsPainted
     True Color=$Particles.Color
     False Color=&ColorFromCurve(ColorCurve=DataInterface() {propertyValues {
        AlphaCurve {keys [<at 0 1>, {arriveTangent=1.6455, interpMode=RCIM_Cubic, leaveTangent=1.6455, time=0.2534, value=5}, <at 0.6077 2>, <at 1 0 RCIM_Cubic>]}
        BlueCurve {keys [<at 0 1.2 RCIM_Cubic>,<at 0.4653 0.4>]}
        GreenCurve {keys [<at 0 0.3 RCIM_Cubic>,<at 0.4653 0.04>]}
        RedCurve {keys [<at 0 1.5 RCIM_Cubic>,<at 0.4653 0.1323>]}
       }})
    ))
   &ScaleRibbonWidth(Ribbon Width Scale=&FloatFromCurve(
     FloatCurve=DataInterface() {propertyValues {Curve {keys [
         {interpMode=RCIM_Cubic, leaveTangentWeight=0.3476, tangentMode=RCTM_User, time=0, value=0.8}
         {arriveTangentWeight=0.3333, interpMode=RCIM_Cubic, tangentMode=RCTM_User, time=0.9, value=3}
        ]}}}
     Scale Curve=&SelectFloatFromArray(Direct Array Index=$Particles.SpawnGroup, Float Selection Array=DataInterface() {propertyValues {FloatData [1,-1]}})
    ))
   /SetVariables(
    Particles.R1 Color:LinearColor=&VC_DynamicScaleColor(
     Color to Scale:LinearColor=$Particles.R1 Color
     RGB Scale:Vector3f=&Multiply_VectorByFloat(
      Vector=<1 1 1>
      Float=&MakeCustomFloatFromBool(
       Bool=$User.bIsPainted
       True Float=&FloatFromCurve(
        FloatCurve=DataInterface() {propertyValues {Curve {keys [{interpMode=RCIM_Cubic, tangentMode=RCTM_User, time=0, value=1}, {interpMode=RCIM_Cubic, tangentMode=RCTM_User, time=1, value=0}]}}}
        Scale Curve=&MakeCustomFloatFromBool(Bool=$User.bIsFrontendPreview, True Float=1.8, False Float=2.8)
       )
       False Float=&FloatFromCurve(
        FloatCurve=DataInterface() {propertyValues {Curve {keys [{interpMode=RCIM_Cubic, tangentMode=RCTM_User, time=0, value=1}, {interpMode=RCIM_Cubic, tangentMode=RCTM_User, time=0.85, value=0}]}}}
        Scale Curve=1.8
       )
      )
     )
     Alpha Scale:NiagaraFloat=&FloatFromCurve(FloatCurve=DataInterface() {propertyValues {Curve {keys [
          {interpMode=RCIM_Cubic, leaveTangentWeight=0.2667, tangentMode=RCTM_User, time=0, value=1}
          {arriveTangentWeight=0.4939, interpMode=RCIM_Cubic, tangentMode=RCTM_User, time=1, value=0}
         ]}}})
    )
    Particles.R2 Color:LinearColor=&VC_DynamicScaleColor(
     Color to Scale:LinearColor=$Particles.R2 Color
     RGB Scale:Vector3f=&Multiply_VectorByFloat(
      Vector=<1 1 1>
      Float=&MakeCustomFloatFromBool(
       Bool=$User.bIsPainted
       True Float=&FloatFromCurve(
        FloatCurve=DataInterface() {propertyValues {Curve {keys [{arriveTangent=-0.3054, interpMode=RCIM_Cubic, leaveTangent=-0.3054, tangentMode=RCTM_User, time=0, value=1.6}, <at 1 1 RCIM_Cubic>]}}}
        Scale Curve=&MakeCustomFloatFromBool(Bool=$User.bIsFrontendPreview, True Float=1, False Float=1.4)
       )
       False Float=&MakeCustomFloatFromBool(
        Bool=$User.bIsFrontendPreview
        True Float=&FloatFromCurve(FloatCurve=DataInterface() {propertyValues {Curve {keys [<at 0 1 RCIM_Cubic>,<at 1 0 RCIM_Cubic>]}}}, Scale Curve=2.5)
        False Float=3
       )
      )
     )
     Alpha Scale:NiagaraFloat=&MakeCustomFloatFromBool(
      Bool=$User.bIsPainted
      True Float=&FloatFromCurve(FloatCurve=DataInterface() {propertyValues {Curve {keys [{arriveTangent=-5.3747, interpMode=RCIM_Cubic, leaveTangent=-5.3747, tangentMode=RCTM_User, time=0, value=1.6}, <at 1 0.3 RCIM_Cubic>]}}})
      False Float=&FloatFromCurve(
       FloatCurve=DataInterface() {propertyValues {Curve {keys [
           <at 0 0.5 RCIM_Cubic>
           {arriveTangent=0.1257, interpMode=RCIM_Cubic, leaveTangent=0.1257, tangentMode=RCTM_User, time=0.1, value=1}
           {arriveTangent=-1.4759, interpMode=RCIM_Cubic, leaveTangent=-1.4759, tangentMode=RCTM_User, time=1, value=0}
          ]}}}
       Scale Curve=0.4
      )
     )
    )
    Particles.RIBBONWIDTH2:NiagaraFloat=&SelectFloatFromArray(Direct Array Index=$Particles.SpawnGroup, Float Selection Array=DataInterface() {propertyValues {FloatData [-40,40]}})
    Particles.RIBBONWIDTH3:NiagaraFloat=&Multiply_Float(
     A=50
     B=&FloatFromCurve(
      FloatCurve=DataInterface() {propertyValues {Curve {keys [
          {interpMode=RCIM_Cubic, leaveTangentWeight=0.1646, tangentMode=RCTM_User, time=0, value=0.5}
          {arriveTangentWeight=0.0707, interpMode=RCIM_Cubic, leaveTangentWeight=0.4007, tangentMode=RCTM_User, time=0.2, value=1}
          <at 1 1>
         ]}}}
      Scale Curve=&SelectFloatFromArray(Direct Array Index=$Particles.SpawnGroup, Float Selection Array=DataInterface() {propertyValues {FloatData [-1,-1]}})
     )
    )
   ) {isSetParameters=true}
  ]
  rendererClasses [&NiagaraRibbonRendererProperties]
  renderers [
   {rendererClass=&NiagaraRibbonRendererProperties}
   {rendererClass=&NiagaraRibbonRendererProperties, rendererIndex=1}
   {rendererClass=&NiagaraRibbonRendererProperties, rendererIndex=2}
  ]
 }
 NE_VC_Boost_Sprite_Parent {
  bEnabled=false
  emitterUpdateScript [
   &EmitterState(
    Inactive Response=::NewEnumerator1
    Life Cycle Mode=::NewEnumerator1
    Loop Duration=1
    Recalculate Duration Each Loop=true
    UseLoopDelay=false
   )
   &SpawnBurst_Instantaneous(Spawn Count=$System.Initial.PortCount, Use Spawn Probability=false, UseLoopCountLimit=false)
   &VC_FilteredSocketCount()
   &SpawnRate(SpawnRate=50, Use Spawn Probability=false) {enabled=false}
  ]
  particleSpawnScript [
   &VC_PaintedVariantsOverride(
    Black=<rgba 0.015 0.015 0.015 1>
    BurntSienna=<rgba 0.1 0.0303 0.005 1>
    Cobalt=<rgba 0 0.0167 1 1>
    Crimson=<rgba 0.85 0 0.0269 1>
    DefaultColor=<rgba 2 1.1245 0.4 1>
    ForestGreen=<rgba 0 0.47 0.0237 1>
    Grey=<rgba 0.065 0.065 0.065 1>
    Lime=<rgba 0.2083 0.5 0 1>
    Orange=<rgba 1.5 0.009 0.009 1>
    Pink=<rgba 0.9 0 0.375 1>
    Purple=<rgba 0.175 0 0.3 1>
    Saffron=<rgba 0.9 0.6 0 1>
    SkyBlue=<rgba 0 0.25 1 1>
    TitaniumWhite=<rgba 0 0 0 1>
   )
   &InitializeParticle(
    Color=$Particles.Color
    Sprite Rotation Mode=::NewEnumerator2
    Sprite Size Mode=::NewEnumerator3
    UsePositionOffset=false
    Write Material Random=true
   )
   &PartitionParticles(Number Of Partitions=$System.PortCount)
   &SkeletalMeshLocation(
    Bone / Socket Index=$Particles.PartitionParticles.Partition
    Bone / Socket Sampling Mode=::NewEnumerator5
    MaskSpawnGroup=false
    Mesh Sampling Type=::NewEnumerator6
    Sampled Position Offset=&AddVectorToPosition(Position=<-105 0 0>, Vector=$User.DTCarOffset)
    Skeletal Mesh=$System.SkeletalMesh
   )
   &AddVelocity(Velocity=<-1500 0 0>) {enabled=false}
   &CameraOffset(Camera Offset Amount=&MakeCustomFloatFromBool(Bool=$User.bIsFrontendPreview, True Float=0, False Float=90))
  ]
  particleUpdateScript [
   &ParticleState()
   &UpdateMeshOrientation(Rotation Rate=&FloatFromCurve(
     FloatCurve=DataInterface() {propertyValues {Curve {keys [
         {interpMode=RCIM_Cubic, tangentMode=RCTM_User, time=0, value=-1}
         {interpMode=RCIM_Cubic, tangentMode=RCTM_User, time=0.5, value=1}
         {tangentMode=RCTM_User, time=1, value=-2}
        ]}}}
     Scale Curve=&Lerp_Float(
      A=&RandomRangeFloat(Minimum=1.8, Maximum=1.5, Evaluation Type=::NewEnumerator1)
      B=&RandomRangeFloat(Minimum=-2.2, Maximum=-1.3, Evaluation Type=::NewEnumerator1)
      Alpha=&MakeFloatFromInt(Int=$Particles.PartitionParticles.Partition)
     )
    ))
   &SolveForcesAndVelocity(Clamp Velocity=false, Limit Acceleration=false)
   &ScaleColor(
    Scale RGB=&Multiply_VectorByFloat(Vector=<1 1 1>, Float=&MakeCustomFloatFromBool(Bool=$User.bIsPainted, True Float=6.5, False Float=1))
    ScaleA=true
    ScaleRGB=true
   )
  ]
  rendererClasses [&NiagaraMeshRendererProperties, &NiagaraSpriteRendererProperties]
  renderers [
   {rendererClass=&NiagaraMeshRendererProperties, rendererIndex=0}
   {rendererClass=&NiagaraSpriteRendererProperties, rendererIndex=1}
  ]
 }
 NE_VC_Boost_Sprite_Parent001 {
  emitterUpdateScript [
   &EmitterState(
    Inactive Response=::NewEnumerator1
    Life Cycle Mode=::NewEnumerator1
    Loop Duration=1
    Recalculate Duration Each Loop=true
    UseLoopDelay=false
   )
   &SpawnBurst_Instantaneous(Spawn Count=$System.Initial.PortCount, Use Spawn Probability=false, UseLoopCountLimit=false)
   &VC_FilteredSocketCount()
   &SpawnRate(SpawnRate=50, Use Spawn Probability=false) {enabled=false}
  ]
  particleSpawnScript [
   &VC_PaintedVariantsOverride(
    Black=<rgba 0.011 0.011 0.011 1>
    BurntSienna=<rgba 0.09 0.0209 0.0018 1>
    Cobalt=<rgba 0 0.0125 0.75 1>
    Crimson=<rgba 0.5 0 0.0158 1>
    DefaultColor=<rgba 2 1.1245 0.4 1>
    ForestGreen=<rgba 0 0.285 0.0095 1>
    Grey=<rgba 0.04 0.04 0.04 1>
    Lime=<rgba 0.125 0.3 0 1>
    Orange=<rgba 0.75 0.0169 0.0045 1>
    Pink=<rgba 0.5 0 0.2083 1>
    Purple=<rgba 0.0875 0 0.15 1>
    Saffron=<rgba 0.5 0.3333 0 1>
    SkyBlue=<rgba 0 0.1625 0.65 1>
    TitaniumWhite=<rgba 0 0 0 1>
   )
   &InitializeParticle(
    Color=$Particles.Color
    Sprite Rotation Mode=::NewEnumerator2
    Sprite Size Mode=::NewEnumerator3
    UsePositionOffset=false
    Write Material Random=true
   )
   &PartitionParticles(Number Of Partitions=$System.PortCount)
   &SkeletalMeshLocation(
    Bone / Socket Index=$Particles.PartitionParticles.Partition
    Bone / Socket Sampling Mode=::NewEnumerator5
    MaskSpawnGroup=false
    Mesh Sampling Type=::NewEnumerator6
    Sampled Position Offset=&AddVectorToPosition(Position=<-105 0 0>, Vector=$User.DTCarOffset)
    Skeletal Mesh=$System.SkeletalMesh
   )
   &AddVelocity(Velocity=<-1500 0 0>) {enabled=false}
   &CameraOffset(Camera Offset Amount=&MakeCustomFloatFromBool(Bool=$User.bIsFrontendPreview, True Float=0, False Float=90))
  ]
  particleUpdateScript [
   &ParticleState()
   &UpdateMeshOrientation(Rotation Rate=&FloatFromCurve(
     FloatCurve=DataInterface() {propertyValues {Curve {keys [
         {interpMode=RCIM_Cubic, tangentMode=RCTM_User, time=0, value=-1}
         {interpMode=RCIM_Cubic, tangentMode=RCTM_User, time=0.5, value=1}
         {tangentMode=RCTM_User, time=1, value=-2}
        ]}}}
     Scale Curve=&Lerp_Float(
      A=&RandomRangeFloat(Minimum=1.8, Maximum=1.5, Evaluation Type=::NewEnumerator1)
      B=&RandomRangeFloat(Minimum=-2.2, Maximum=-1.3, Evaluation Type=::NewEnumerator1)
      Alpha=&MakeFloatFromInt(Int=$Particles.PartitionParticles.Partition)
     )
    ))
   &SolveForcesAndVelocity(Clamp Velocity=false, Limit Acceleration=false)
   &ScaleColor(
    Scale RGB=&Multiply_VectorByFloat(Vector=<1 1 1>, Float=&MakeCustomFloatFromBool(Bool=$User.bIsPainted, True Float=6.5, False Float=1))
    ScaleA=true
    ScaleRGB=true
   )
   &ShapeLocation(
    MaskSpawnGroup=false
    Shape Origin=&AddVectorToPosition(
     Position=$Particles.Initial.Position
     Vector=&MakeVector(
      X=0
      Y=&Sine(Normalized Angle=$Emitter.Age, Period=1.15, Scale=1, Bias=1)
      Z=&Sine(Normalized Angle=$Emitter.Age, Period=0.75, Scale=3, Bias=0)
     )
    )
    Sphere Distribution=::NewEnumerator1
    Sphere Radius=0
    UseOwnerScale=true
   )
  ]
  rendererClasses [&NiagaraMeshRendererProperties, &NiagaraSpriteRendererProperties]
  renderers [
   {rendererClass=&NiagaraMeshRendererProperties, rendererIndex=0}
   {rendererClass=&NiagaraSpriteRendererProperties, rendererIndex=1}
  ]
 }
}
summary {
 systemName=NS_Boost_Dagger_Backup
 emitters [
  {emitterName=NE_VC_Boost_RIBBONS, rendererClasses [&NiagaraRibbonRendererProperties]}
  {
   emitterName=NE_VC_Boost_BIGCONE002
   rendererClasses [&NiagaraMeshRendererProperties, &NiagaraSpriteRendererProperties]
  }
  {
   emitterName=NE_VC_Boost_Sprite_Parent001
   rendererClasses [&NiagaraMeshRendererProperties, &NiagaraSpriteRendererProperties]
  }
  {
   bEnabled=false
   emitterName=NE_VC_Boost_BIGCONE
   rendererClasses [&NiagaraMeshRendererProperties, &NiagaraSpriteRendererProperties]
  }
  {
   bEnabled=false
   emitterName=NE_VC_Boost_Sprite_Parent
   rendererClasses [&NiagaraMeshRendererProperties, &NiagaraSpriteRendererProperties]
  }
 ]
 userVariables [
  User.bIsFrontendPreview {type=NiagaraBool}
  User.bIsPainted {type=NiagaraBool}
  User.ColorVariantID {type=NiagaraInt32}
  User.DTCarOffset {type=Vector3f, description="", defaultValue {struct=Vector3f, value=<0 0 0>}}
  User.PaintedColor {type=LinearColor, description="", defaultValue {struct=LinearColor, value=<rgba 1 1 1 1>}}
  User.PaintedScalar {type=NiagaraFloat, defaultValue=1, description=""}
  User.PortCount {type=NiagaraInt32, defaultValue=2}
 ]
}
