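# Translucent, two-sided unlit material for Niagara mesh particles, apparently the airwave effect of the Dagger vehicle boost.
# A panning wind texture times a gradient mask is tinted by particle color and fed to an emissive adjustment function.
# The same mask, sharpened, combines with particle alpha and a dynamic Opacity parameter to form an opacity term.
# A Speed-driven sine over world X along vertex normals, faded by the V coordinate, forms a wobble offset.
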Material /VehicleCosmetics/Boosts/Boost_Dagger/Materials/M_Boost_AirwaveM_Boostat
graph {
 properties {
  bAutomaticallySetUsageInEditor=false
  bUsedWithNiagaraMeshParticles=true
  blendMode=BLEND_Translucent
  shadingModel=MSM_Unlit
  twoSided=true
  usedShadingModels=Unlit
 }
 nodes [
  MaterialExpressionMeshEmitterVertexColor_4 = ParticleColor()
  n1 = TextureCoordinate() {vTiling=2}
  n2 = Multiply(
   A=TextureSample(UVs=Panner(Coordinate=@n1) {speedX=1, speedY=-3}) {texture=/DealershipSharedLibrary/FX/Textures/Gradient_Masks/T_VCVFX_Wind_A}.RGB
   B=Power(Base=TextureSample() {texture=/DealershipSharedLibrary/FX/Textures/Gradient_Masks/T_VCVFX_Gradient_A}.RGB, Exp=Constant() {r=1.5})
  )
  MaterialFunctionCall(Emissive=Multiply(A=Multiply(A=@MaterialExpressionMeshEmitterVertexColor_4.RGB, B=@n2)) {constB=3}) {materialFunction=&MF_FX_EmissiveAdjustments}
  n3 = Power(Base=@n2, Exp=Constant() {r=4})
  MaterialExpressionMeshEmitterDynamicParameter_4 = DynamicParameter() {paramNames [Opacity,Param2,Param3,Param4]}
  MaterialExpressionConstantClamp_7 = Saturate(None=Multiply(A=Multiply(A=@n3, B=@MaterialExpressionMeshEmitterVertexColor_4.A), B=@MaterialExpressionMeshEmitterDynamicParameter_4.Opacity))
  Multiply(A=@MaterialExpressionConstantClamp_7) {constB=2}
  n4 = WorldPosition()
  n5 = Multiply(A=Time(), B=ScalarParameter() {defaultValue=5, parameterName=Speed})
  n6 = Sine(None=ComponentMask(None=Add(A=@n4.XYZ, B=@n5)) {r=true})
  MaterialExpressionWorldNormal_3 = VertexNormalWS()
  n7 = TextureCoordinate()
  MaterialExpressionConstantClamp_8 = Saturate(None=Multiply(A=Constant() {r=2}, B=ComponentMask(None=@n7) {g=true}))
  Multiply(A=Multiply(A=Multiply(A=@n6, B=@MaterialExpressionWorldNormal_3), B=Constant() {r=1}), B=@MaterialExpressionConstantClamp_8)
 ]
}
