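# Translucent, unlit, two-sided material for Niagara mesh particles in the Dagger vehicle boost cosmetic.
# Per-axis drift comes from time-driven sine waves blended by further sine or offset cosine waves, each with tunable speeds.
# The drift is scaled per axis and uniformly, then added to a vector parameter named as a world position offset.
# Particle color times a dynamic parameter's red channel feeds MF_FX_EmissiveAdjustments for emissive.
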Material /VehicleCosmetics/Boosts/Boost_Dagger/Materials/M_Boost_Dagger
graph {
 properties {
  bAutomaticallySetUsageInEditor=false
  bUsedWithNiagaraMeshParticles=true
  blendMode=BLEND_Translucent
  shadingModel=MSM_Unlit
  twoSided=true
  usedShadingModels=Unlit
 }
 nodes [
  n1 = ScalarParameter() {defaultValue=1.5, parameterName=SpeedDB}
  n2 = Time()
  n3 = ScalarParameter() {defaultValue=1, parameterName=SpeedAB}
  n4 = Time()
  n5 = LinearInterpolate(A=Constant() {r=1}, B=Sine(None=Multiply(A=@n1, B=@n2)), Alpha=Sine(None=Multiply(A=@n3, B=@n4)))
  n6 = ScalarParameter() {defaultValue=1, parameterName=X_Drift_Mult}
  n7 = ScalarParameter() {defaultValue=1.5, parameterName=SpeedD}
  n8 = Time()
  n9 = ScalarParameter() {defaultValue=1, parameterName=SpeedA}
  n10 = Time()
  n11 = LinearInterpolate(A=Constant() {r=1}, B=Sine(None=Multiply(A=@n7, B=@n8)), Alpha=Sine(None=Multiply(A=@n9, B=@n10)))
  n12 = ScalarParameter() {defaultValue=1, parameterName=Y_Drift_Mult}
  n13 = Sine(None=Multiply(A=ScalarParameter() {defaultValue=1.5, parameterName=SpeedC}, B=Time()))
  n14 = Constant() {r=-1}
  n15 = Time()
  n16 = ScalarParameter() {defaultValue=1, parameterName=TimeOffset}
  n17 = Cosine(None=Multiply(A=ScalarParameter() {defaultValue=1, parameterName=SpeedB}, B=Add(A=@n15, B=@n16)))
  n18 = AppendVector(
   A=AppendVector(A=Multiply(A=@n5, B=@n6), B=Multiply(A=@n11, B=@n12))
   B=Multiply(A=LinearInterpolate(A=@n13, B=@n14, Alpha=@n17), B=ScalarParameter() {defaultValue=1, parameterName=Z_Drift_Mult})
  )
  n19 = ScalarParameter() {defaultValue=5, parameterName=Uniform_Drift_Mult}
  Add(
   A=ComponentMask(None=Multiply(A=@n18, B=@n19)) {b=true, g=true, r=true}
   B=ComponentMask(None=VectorParameter() {parameterName=WPO_OffsetB}.RGB) {b=true, g=true, r=true}
  )
  MaterialExpressionMeshEmitterDynamicParameter_5 = DynamicParameter()
  MaterialExpressionMeshEmitterVertexColor_5 = ParticleColor()
  MaterialFunctionCall(Emissive=Multiply(A=@MaterialExpressionMeshEmitterDynamicParameter_5.R, B=@MaterialExpressionMeshEmitterVertexColor_5.RGB)) {materialFunction=&MF_FX_EmissiveAdjustments}
 ]
}
