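# Translucent unlit Niagara ribbon material for the Dagger vehicle boost's smoke and wind trails.
# Static switches, separate for emissive and alpha, select a wind trail of panning helix and speedline masks, a main smoke ribbon, or a secondary smoke ribbon with noise-distorted UVs.
# Smoke layers multiply panning smoke and cloud noise, and UV gradients fade the ribbon ends.
# Particle color and dynamic parameters drive tint, intensity and offsets, with depth fade softening intersections.
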Material /VehicleCosmetics/Boosts/Boost_Dagger/Materials/M_Boost_NewSM_Boostoketrail
graph {
 properties {
  bAutomaticallySetUsageInEditor=false
  bUsedWithNiagaraRibbons=true
  blendMode=BLEND_Translucent
  shadingModel=MSM_Unlit
  usedShadingModels=Unlit
 }
 nodes [
  DepthFade() {fadeDistanceDefault=250}
  n1 = TextureCoordinate()
  n2 = DynamicParameter() {parameterIndex=1, defaultValue=<rgba 1 1 0 0>, paramNames [POWER,DISTORSION,OFFSETX,OFFSETY]}
  n3 = MaterialFunctionCall(X=@n2.OFFSETX, Y=@n2.OFFSETX) {materialFunction=&MakeFloat2}
  n4 = Panner(Coordinate=TextureCoordinate() {uTiling=0.75, vTiling=0.4}) {speedX=-2}
  n5 = ComponentMask(None=TextureSample(UVs=@n4) {texture=/DealershipSharedLibrary/FX/Textures/Noise/T_VCVFX_TrailMask_RGB}.R) {g=true, r=true}
  n6 = Panner(Coordinate=Add(A=Add(A=@n1, B=@n3.Result), B=Multiply(A=@n5, B=@n2.DISTORSION))) {speedX=-0.5}
  MaterialFunctionCall(
   UVs=@n6
   Texture Scale=MaterialFunctionCall(X=Constant() {r=1}, Y=ScalarParameter() {defaultValue=1, parameterName=Y SCALE FROM RIBBON 1}) {materialFunction=&MakeFloat2}.Result
  ) {materialFunction=&ScaleUVsByCenter}
  MaterialFunctionCall(
   UVs=@n4
   Texture Scale=MaterialFunctionCall(X=Constant() {r=1}, Y=ScalarParameter() {defaultValue=1, parameterName=Y SCALE FROM RIBBON 1}) {materialFunction=&MakeFloat2}.Result
  ) {materialFunction=&ScaleUVsByCenter}
  n7 = ParticleColor()
  n8 = Multiply(A=TextureCoordinate() {vTiling=0.2}, B=ScalarParameter() {defaultValue=1, parameterName=texcoord})
  n9 = Multiply(A=Time(), B=ScalarParameter() {defaultValue=1, parameterName=speed01})
  n10 = DynamicParameter() {parameterIndex=2, paramNames [Param1,OFFSETX,OFFSETY,Param4]}
  n11 = MaterialFunctionCall(X=@n10.OFFSETX, Y=@n10.OFFSETX) {materialFunction=&MakeFloat2}
  n12 = TextureSample(UVs=Add(A=Panner(Coordinate=@n8, Time=@n9) {speedX=-1}, B=@n11.Result)) {texture=/DealershipSharedLibrary/FX/Textures/Gradient_Masks/T_VCVFX_Boost_Helix_WindParticles_Alpha}
  n13 = Multiply(A=TextureCoordinate(), B=ScalarParameter() {defaultValue=1, parameterName=texcoord})
  n14 = Multiply(A=Time(), B=ScalarParameter() {defaultValue=1, parameterName=speed02})
  n15 = TextureSample(UVs=Add(A=Panner(Coordinate=@n13, Time=@n14) {speedX=-3}, B=@n11.Result)) {texture=/DealershipSharedLibrary/FX/Textures/Gradient_Masks/T_VCVFX_speedline_A}
  n16 = TextureCoordinate() {uTiling=2}
  n17 = TextureCoordinate()
  n18 = Multiply(A=Multiply(A=ComponentMask(None=@n16) {r=true}) {constB=4}, B=OneMinus(None=ComponentMask(None=@n17) {r=true})) {constB=5}
  n19 = ScalarParameter() {defaultValue=1, parameterName=opa}
  n20 = Multiply(A=@n7.A, B=Multiply(A=Add(A=@n12.R, B=@n15.R), B=Power(Base=@n18, Exp=@n19)))
  MaterialExpressionVertexColor_4 = ParticleColor()
  n21 = TextureCoordinate() {uTiling=2}
  n22 = Panner(Coordinate=TextureCoordinate() {uTiling=0.8, vTiling=0.8}) {speedX=-0.75, speedY=0.1}
  n23 = Multiply(
   A=TextureSample(UVs=Panner(Coordinate=@n21) {speedX=-1.2}) {texture=/DealershipSharedLibrary/FX/Textures/TexturePackage/T_VCVFX_SmokeTrail01_RGB}.R
   B=Power(Base=TextureSample(UVs=@n22) {texture=/DealershipSharedLibrary/FX/Textures/Noise/T_VCVFX_CloudDistort_A}.RGB, Exp=Constant() {r=0.8})
  )
  n24 = TextureCoordinate()
  MaterialExpressionConstantClamp_13 = Saturate(None=Multiply(A=ComponentMask(None=@n24) {r=true}, B=Constant() {r=20}))
  n25 = Power(Base=@MaterialExpressionConstantClamp_13, Exp=Constant() {r=2})
  n26 = ComponentMask(None=TextureCoordinate()) {r=true}
  MaterialExpressionConstantClamp_14 = Saturate(None=Multiply(A=OneMinus(None=@n26), B=Constant() {r=10}))
  n27 = Power(Base=@MaterialExpressionConstantClamp_14, Exp=Constant() {r=4})
  MaterialExpressionConstantClamp_12 = Saturate(None=Multiply(A=Multiply(A=@MaterialExpressionVertexColor_4.A, B=@n23), B=Multiply(A=@n25, B=@n27)))
  n28 = OneMinus(None=ComponentMask(None=TextureCoordinate()) {r=true})
  n29 = Multiply(A=@MaterialExpressionConstantClamp_12, B=Multiply(A=Power(Base=@n28) {constExponent=1.2}) {constB=2})
  MaterialExpressionVertexColor_3 = ParticleColor()
  n30 = ScalarParameter() {defaultValue=1, parameterName=texcoord}
  n31 = Panner(Coordinate=TextureCoordinate() {uTiling=0.5, vTiling=0.5}) {speedX=-0.8, speedY=0.2}
  n32 = Multiply(
   A=TextureSample(UVs=Multiply(A=@n6, B=@n30)) {texture=&T_VCVFX_SmokeTrail01_RGB}.R
   B=Power(Base=TextureSample(UVs=@n31) {texture=&T_VCVFX_CloudDistort_A}.RGB, Exp=Constant() {r=1})
  )
  n33 = Multiply(A=@MaterialExpressionVertexColor_3.A, B=@n32)
  n34 = TextureCoordinate()
  MaterialExpressionConstantClamp_10 = Saturate(None=Multiply(A=ComponentMask(None=@n34) {r=true}, B=Constant() {r=20}))
  n35 = ComponentMask(None=TextureCoordinate()) {r=true}
  MaterialExpressionConstantClamp_11 = Saturate(None=Multiply(A=OneMinus(None=@n35), B=Constant() {r=10}))
  n36 = Multiply(
   A=Power(Base=@MaterialExpressionConstantClamp_10, Exp=Constant() {r=2})
   B=Power(Base=@MaterialExpressionConstantClamp_11, Exp=Constant() {r=2})
  )
  n37 = ComponentMask(None=TextureCoordinate()) {g=true}
  n38 = Multiply(A=Power(Base=OneMinus(None=@n37)) {constExponent=3}) {constB=5}
  MaterialExpressionConstantClamp_9 = Saturate(None=Multiply(A=Multiply(A=@n33, B=@n36), B=Saturate(None=@n38)))
  MaterialFunctionCall(
   Opacity In=StaticSwitchParameter(
    True=Saturate(None=@n20)
    False=StaticSwitchParameter(True=@n29, False=@MaterialExpressionConstantClamp_9) {parameterName=IS MAIN RIBBON ALPHA?}
   ) {parameterName=IS WINDTRAIL ALPHA?}
   DepthFade Distance=Constant() {r=50}
  ) {materialFunction=&MF_VCVFX_DepthFade}
  n39 = Multiply(A=DynamicParameter().R, B=Multiply(A=@MaterialExpressionVertexColor_4.RGB, B=@n23))
  n40 = Multiply(A=Multiply(A=@MaterialExpressionVertexColor_3.RGB, B=@n32), B=@n2.POWER)
  MaterialFunctionCall(Emissive=StaticSwitchParameter(True=Multiply(A=@n10.Param1, B=@n7.RGB), False=StaticSwitchParameter(True=@n39, False=@n40) {parameterName=IS MAIN RIBBON?}) {parameterName=IS WINDTRAIL?}) {materialFunction=&MF_FX_EmissiveAdjustments}
 ]
}
